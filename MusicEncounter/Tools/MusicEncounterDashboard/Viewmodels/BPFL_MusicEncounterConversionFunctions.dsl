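; Blueprint function library of conversion helpers for the Music Encounter dashboard tool's viewmodels.
; Encounter states are packed in an integer bitmask, where each enum value maps to a power-of-two flag tested with bitwise AND.
; GetMusicEncounterViewModels reduces each viewmodel's flags to one primary state by fixed priority with a fallback, returning those matching the requested state.
; SelectSlateColor wraps a boolean choice between two colors as a SlateColor.
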(fn GetBitFlag
  (params (State :type EMusicEncounterState%20Enum :default "NewEnumerator1"))
  (return (Math|Float|Round (Math|Float|Power 2.0 (Math|Conversions|ToFloat(Byte) State)))))

(fn GetMusicEncounterStateValue
  (params (State%20Type :type EMusicEncounterState%20Enum :default "NewEnumerator1") (State :type Integer))
  (return (Math|Conversions|ToBoolean(Integer) (Math|Integer|BitwiseAND (MusicEncounter|GetBitFlag State%20Type) State))))

(fn GetMusicEncounterViewModels
  (params (Music%20Encounter%20ViewModels :type Array%20of%20Music%20Encounter%20View%20Model%20Object%20References) (State :type EMusicEncounterState%20Enum :default "NewEnumerator1"))
  (bind _music_encounter_viewmodel (LocalVariables|GetMusic%20Encounter%20ViewModel))
  (bind _music_encounter_state (Default|GetMusicEncounterState _music_encounter_viewmodel))
  (for _array_element Music%20Encounter%20ViewModels
    (LocalVariables|SetMusic%20Encounter%20ViewModel _array_element)
    (if (MusicEncounter|GetMusicEncounterStateValue "NewEnumerator3" _music_encounter_state _music_encounter_viewmodel)
      (LocalVariables|SetPrimary%20State "NewEnumerator3")
      (elif (MusicEncounter|GetMusicEncounterStateValue "NewEnumerator2" _music_encounter_state _music_encounter_viewmodel)
        (LocalVariables|SetPrimary%20State "NewEnumerator2")
        (elif (MusicEncounter|GetMusicEncounterStateValue "NewEnumerator1" _music_encounter_state _music_encounter_viewmodel)
          (LocalVariables|SetPrimary%20State "NewEnumerator1")
          (else
            (LocalVariables|SetPrimary%20State "NewEnumerator8")))))
    (if (Utilities|Enum|Equal(Enum) (Variables|GetState) (LocalVariables|GetPrimary%20State))
      (Utilities|Array|Add (LocalVariables|GetViewModels) (LocalVariables|GetMusic%20Encounter%20ViewModel))))
  (return (LocalVariables|GetViewModels)))

(fn SelectSlateColor
  (params (Color%20A :type Linear%20Color%20Structure) (Color%20B :type Linear%20Color%20Structure) (Select%20A :type Boolean))
  (return (Utilities|Struct|MakeSlateColor (Math|Color|SelectColor Color%20A Color%20B Select%20A))))
