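; Viewmodel for the MusicEncounter dashboard tool that exposes an audio component's sound name for display.
; SetAudioComponent checks whether the passed component is valid and, if so, reads the sound from the stored Audio Component variable.
; The sound's display name becomes the Sound Name text, which is set empty when the component is invalid.
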(fn SetAudioComponent
  (params (Audio%20Component :type Audio%20Component%20Object%20Reference))
  (Utilities|IsValid Audio%20Component
    (:Is%20Valid
      (bind _sound (Variables|/Script/Engine.AudioComponent|GetSound (Variables|GetAudio%20Component))
        (:then
          (Variables|Default|SetSound%20Name (Utilities|Text|ToText(String) (Utilities|GetDisplayName _sound))))
        (:else)))
    (:Is%20Not%20Valid
      (Variables|Default|SetSound%20Name ""))))
