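; View model for the Music Encounter dashboard tool that exposes one client's encounters as per-encounter view models.
; Initialization wraps the manager component's existing encounters and binds to newly constructed ones, suppressing the added dispatcher until finished.
; Encounter state and highest-priority music event changes rebroadcast the list through FieldNotify.
; Priority sorting is still a todo, and Reset, called for an invalid manager, is empty.
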(event Custom|Music%20Encounter%20State%20Changed
  (params (Object :type Object%20Reference) (Field :type Field%20Notification%20Id%20Structure))
  (FieldNotification|BroadcastFieldValueChanged self "(FieldName=\"Music Encounter ViewModels\")"))

(fn AddMusicEncounter
  (params (Music%20Encounter :type Music%20Encounter%20Base%20Object%20Reference))
  (Utilities|IsValid Music%20Encounter
    (:Is%20Valid
      (Utilities|FlowControl|Sequence
        (:then_0
          (bind _returnvalue (Game|ConstructObjectfromClass "/MusicEncounter/Tools/MusicEncounterDashboard/Viewmodels/BP_MusicEncounterViewModel.BP_MusicEncounterViewModel_C"))
          (bind _output_get (LocalVariables|SetMusic%20Encounter%20ViewModel _returnvalue))
          (FieldNotify|AddFieldValueChangedDelegate _output_get "(FieldName=\"Music Encounter State\")" (EventDispatchers|CreateEvent :!event Music%20Encounter%20State%20Changed)))
        (:then_1
          (Utilities|FlowControl|Sequence
            (:then_0
              (Default|SetMusicEncounter (LocalVariables|GetMusic%20Encounter%20ViewModel) (Variables|GetMusic%20Encounter))
              (Utilities|Array|AddUnique (Variables|Default|GetMusic%20Encounter%20ViewModels) (LocalVariables|GetMusic%20Encounter%20ViewModel)))
            (:then_1
              (Utilities|FlowControl|Sequence
                (:then_0
                  (if (Variables|Default|GetIs%20Initialized)
                    (Default|CallOnMusicEncounterAdded self)))))))))))

(fn GetMusicEncounterViewModels
  ;[1 1 1] @todo: sort this list by music event priority
  (return (Variables|Default|GetMusic%20Encounter%20ViewModels)))

(fn InitializeClientViewModel
  (params (Music%20Encounter%20Manager%20Component :type Music%20Encounter%20Manager%20Component%20Object%20Reference))
  (Variables|Default|SetIs%20Initialized false)
  (Utilities|IsValid Music%20Encounter%20Manager%20Component
    (:Is%20Valid
      (Utilities|FlowControl|Sequence
        (:then_0
          (MusicEncounterManager|BindEventtoOnMusicEncounterConstructed (Variables|GetMusic%20Encounter%20Manager%20Component) (EventDispatchers|CreateEvent :!event Add%20Music%20Encounter))
          (EventDispatchers|BindEventtoOnHighestPriorityMusicEventChanged (GameInstanceSubsystems|GetMusicEventSubsystem (Variables|GetMusic%20Encounter%20Manager%20Component)) (EventDispatchers|CreateEvent :!event Update%20Music%20Encounter%20List)))
        (:then_1
          (Utilities|FlowControl|Sequence
            (:then_0
              (bind _values (Utilities|Map|Values (Variables|/MusicEncounter/BP_MusicEncounterManager.BP_MusicEncounterManager_C|GetMusic%20Encounters (Variables|GetMusic%20Encounter%20Manager%20Component))))
              (for _array_element _values
                (Default|AddMusicEncounter _array_element)))
            (:then_1
              (Utilities|FlowControl|Sequence
                (:then_0
                  (FieldNotification|BroadcastFieldValueChanged self "(FieldName=\"Music Encounter ViewModels\")"))
                (:then_1
                  (Utilities|FlowControl|Sequence
                    (:then_0
                      (Variables|Default|SetIs%20Initialized true))))))))))
    (:Is%20Not%20Valid
      (Default|Reset))))

(fn Reset)

(fn UpdateMusicEncounterList
  (FieldNotification|BroadcastFieldValueChanged self "(FieldName=\"Music Encounter ViewModels\")"))
