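; Viewmodel for a music encounter component in the Music Encounter Dashboard tool, exposing the component's name as text.
; SetMusicEncounterComponent checks the component passed in and, when it is valid, fills Component Name from the component's display name.
; An invalid component clears Component Name to empty text.
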(fn SetMusicEncounterComponent
  (params (Music%20Encounter%20Component :type Music%20Encounter%20Component%20Base%20Object%20Reference))
  (Utilities|IsValid Music%20Encounter%20Component
    (:Is%20Valid
      (Variables|Default|SetComponent%20Name (Utilities|Text|ToText(String) (Utilities|GetDisplayName (Variables|GetMusic%20Encounter%20Component)))))
    (:Is%20Not%20Valid
      (Variables|Default|SetComponent%20Name ""))))
