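; Dashboard view model for the MusicEncounter tool that tracks clients and lists the selected client's music encounters.
; InitializeClients creates a client view model per new manager component and binds its encounter events.
; Those events rebroadcast the encounter-list field notification when the selected client changes, refreshing bound widgets.
; The list is grouped active, paused, initialized, then uninitialized, with a Music Event subsystem priority query used to reorder active encounters.
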(event Custom|Update%20Displayed%20ViewModels
  (params (Client%20ViewModel :type BP%20Music%20Encounter%20Client%20View%20Model%20Object%20Reference))
  ;[1 1 1] Updates the widget if the selected client's music encounter viewmodels have updated
  (if (Utilities|Array|IsValidIndex (Variables|Clients|GetClient%20ViewModels) (Variables|Clients|GetSelected%20Client))
    (if (== Client%20ViewModel (Utilities|Array|Get(acopy) (Variables|Clients|GetClient%20ViewModels) (Variables|Clients|GetSelected%20Client)))
      (FieldNotification|BroadcastFieldValueChanged self "(FieldName=\"Get Music Encounter ViewModels for Selected Client\")"))))

(event Custom|Music%20Encounter%20State%20Changed
  (params (Object :type Object%20Reference) (Field :type Field%20Notification%20Id%20Structure))
  (bind _asbp_music_encounter_client_view_model (Utilities|Casting|CastToBP_MusicEncounterClientViewModel Object))
  (CallFunction|UpdateDisplayedViewModels _asbp_music_encounter_client_view_model))

(fn GetMusicEncounterViewModelsforSelectedClient
  (bind _selected_client (Variables|Clients|GetSelected%20Client))
  (bind _client_viewmodels (Variables|Clients|GetClient%20ViewModels))
  (bind _output (Utilities|Array|Get(acopy) _client_viewmodels _selected_client))
  (if (Utilities|Array|IsValidIndex (Variables|Clients|GetClient%20ViewModels) (Variables|Clients|GetSelected%20Client))
    (Utilities|IsValid _output
      (:Is%20Valid
        (return (Clients|SortDisplayedMusicEncounterViewModels (Class|BPMusicEncounterClientViewModel|GetMusicEncounterViewModels _output))))
      (:Is%20Not%20Valid
        (return 0)))
    (else
      (return 0))))

(fn GetNumberofClients
  (return (Utilities|Array|Length (Variables|Clients|GetClient%20ViewModels))))

(fn GetSelectedClientMusicEncounterManagerComponent
  (if (Utilities|Array|IsValidIndex (Variables|Clients|GetMusic%20Encounter%20Manager%20Components) (Variables|Clients|GetSelected%20Client))
    (return (Utilities|Array|Get(acopy) (Variables|Clients|GetMusic%20Encounter%20Manager%20Components) (Variables|Clients|GetSelected%20Client)))))

(fn InitializeClients
  (params (Music%20Encounter%20Manager%20Components :type Array%20of%20Music%20Encounter%20Manager%20Component%20Object%20References))
  (for _array_element Music%20Encounter%20Manager%20Components
    (bind _output_get (LocalVariables|SetMusic%20Encounter%20Manager%20Component _array_element))
    (if (not (Utilities|Array|ContainsItem (Variables|Clients|GetMusic%20Encounter%20Manager%20Components) _output_get))
      (Utilities|FlowControl|Sequence
        (:then_0
          (bind _returnvalue (Game|ConstructObjectfromClass "/MusicEncounter/Tools/MusicEncounterDashboard/Viewmodels/BP_MusicEncounterClientViewModel.BP_MusicEncounterClientViewModel_C"))
          (LocalVariables|SetClient%20ViewModel _returnvalue))
        (:then_1
          (Utilities|FlowControl|Sequence
            (:then_0
              (Class|BPMusicEncounterClientViewModel|InitializeClientViewModel (LocalVariables|GetClient%20ViewModel) (LocalVariables|GetMusic%20Encounter%20Manager%20Component)))
            (:then_1
              (Utilities|FlowControl|Sequence
                (:then_0
                  (Utilities|Array|Add (Variables|Clients|GetClient%20ViewModels) (LocalVariables|GetClient%20ViewModel)))
                (:then_1
                  (Utilities|FlowControl|Sequence
                    (:then_0
                      (Utilities|Array|Add (Variables|Clients|GetMusic%20Encounter%20Manager%20Components) (LocalVariables|GetMusic%20Encounter%20Manager%20Component))
                      (LocalVariables|SetAdded%20New%20Client true))
                    (:then_1
                      (Utilities|FlowControl|Sequence
                        (:then_0
                          (Default|BindEventtoOnMusicEncounterAdded (LocalVariables|GetClient%20ViewModel) (EventDispatchers|CreateEvent :!event Update%20Displayed%20ViewModels))
                          (FieldNotify|AddFieldValueChangedDelegate (LocalVariables|GetClient%20ViewModel) "(FieldName=\"Music Encounter ViewModels\")" (EventDispatchers|CreateEvent :!event Music%20Encounter%20State%20Changed))))))))))))))
  (if (LocalVariables|GetAdded%20New%20Client)
    (Utilities|FlowControl|Sequence
      (:then_0
        (FieldNotification|BroadcastFieldValueChanged self "(FieldName=\"Get Music Encounter ViewModels for Selected Client\")"))
      (:then_1
        (Utilities|FlowControl|Sequence
          (:then_0
            (FieldNotification|BroadcastFieldValueChanged self "(FieldName=\"Get Number of Clients\")")))))))

(fn Reset
  (Utilities|Array|Clear (Variables|Clients|GetClient%20ViewModels))
  (Utilities|Array|Clear (Variables|Clients|GetMusic%20Encounter%20Manager%20Components))
  (Variables|Clients|SetSelected%20Client 0))

(fn SortDisplayedMusicEncounterViewModels
  (params (Music%20Encounter%20ViewModels :type Array%20of%20Music%20Encounter%20View%20Model%20Object%20References))
  (bind _sorted_viewmodels (LocalVariables|GetSorted%20ViewModels))
  (for _array_element Music%20Encounter%20ViewModels
    (LocalVariables|SetMusic%20Encounter _array_element)
    (for _array_element_2 (LocalVariables|GetState%20Types)
      (switch Utilities|FlowControl|Switch|SwitchonEMusicEncounterState _array_element_2
        (:NewEnumerator8
          (if (not (Default|GetMusicEncounterStateValue (LocalVariables|GetMusic%20Encounter) "NewEnumerator1"))
            (Utilities|Array|Add (LocalVariables|GetUninitialized%20Encounters) (LocalVariables|GetMusic%20Encounter))))
        (:NewEnumerator1
          (if (Math|Boolean|ANDBoolean (Default|GetMusicEncounterStateValue (LocalVariables|GetMusic%20Encounter) "NewEnumerator1") (not (Default|GetMusicEncounterStateValue (LocalVariables|GetMusic%20Encounter) "NewEnumerator2")) (not (Default|GetMusicEncounterStateValue (LocalVariables|GetMusic%20Encounter) "NewEnumerator3")))
            (Utilities|Array|Add (LocalVariables|GetInitialized%20Encounters) (LocalVariables|GetMusic%20Encounter))))
        (:NewEnumerator2
          (if (Math|Boolean|ANDBoolean (Default|GetMusicEncounterStateValue (LocalVariables|GetMusic%20Encounter) "NewEnumerator1") (Default|GetMusicEncounterStateValue (LocalVariables|GetMusic%20Encounter) "NewEnumerator2") (not (Default|GetMusicEncounterStateValue (LocalVariables|GetMusic%20Encounter) "NewEnumerator3")))
            (Utilities|Array|Add (LocalVariables|GetActive%20Encounters) (LocalVariables|GetMusic%20Encounter))))
        (:NewEnumerator3
          (if (Math|Boolean|ANDBoolean (Default|GetMusicEncounterStateValue (LocalVariables|GetMusic%20Encounter) "NewEnumerator1") (not (Default|GetMusicEncounterStateValue (LocalVariables|GetMusic%20Encounter) "NewEnumerator2")) (Default|GetMusicEncounterStateValue (LocalVariables|GetMusic%20Encounter) "NewEnumerator3"))
            (Utilities|Array|Add (LocalVariables|GetPaused%20Encounters) (LocalVariables|GetMusic%20Encounter)))))))
  (bind _output_get (LocalVariables|SetActive%20Encounters (Clients|SortHighestActiveEncounter (LocalVariables|GetActive%20Encounters))))
  (Utilities|Array|AppendArray _sorted_viewmodels _output_get)
  (Utilities|Array|AppendArray _sorted_viewmodels (LocalVariables|GetPaused%20Encounters))
  (Utilities|Array|AppendArray _sorted_viewmodels (LocalVariables|GetInitialized%20Encounters))
  (Utilities|Array|AppendArray _sorted_viewmodels (LocalVariables|GetUninitialized%20Encounters))
  (return (LocalVariables|GetSorted%20ViewModels)))

(fn SortHighestActiveEncounter
  (params (In :type Array%20of%20Music%20Encounter%20View%20Model%20Object%20References))
  (for _array_element In
    (if (== 0 (Audio|GetHighestPriorityInstance (GameInstanceSubsystems|GetMusicEventSubsystem (Clients|GetSelectedClientMusicEncounterManagerComponent))))
      (LocalVariables|SetHighest%20Priority%20Music%20Encounter _array_element)))
  (bind _returnvalue (Utilities|Array|RemoveItem (Variables|GetIn) (LocalVariables|GetHighest%20Priority%20Music%20Encounter)))
  (if _returnvalue
    (Utilities|Array|Insert (Variables|GetIn) (LocalVariables|GetHighest%20Priority%20Music%20Encounter))
    (else
      _))
  (return (Variables|GetIn)))
