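; Viewmodel for the music encounter manager in the Music Encounter Dashboard tool, exposing a list of per-encounter viewmodels and a selected one.
; Setting the manager clears the list, then, given a valid manager, binds to its encounter-constructed event and creates a viewmodel for each existing encounter.
; Each creation broadcasts a field change for the list.
; After the rebuild, the first entry becomes selected, or none if empty.
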(fn CreateMusicEncounterViewmodel
  (params (Music%20Encounter :type Music%20Encounter%20Base%20Object%20Reference))
  (bind _returnvalue (Game|ConstructObjectfromClass "/MusicEncounter/Tools/MusicEncounterDashboard/Viewmodels/BP_MusicEncounterViewModel.BP_MusicEncounterViewModel_C"))
  (Default|SetMusicEncounter _returnvalue Music%20Encounter)
  (Utilities|Array|Add (Variables|Default|GetMusic%20Encounter%20Viewmodels) _returnvalue)
  (FieldNotification|BroadcastFieldValueChanged self "(FieldName=\"Music Encounter Viewmodels\")"))

(fn SetMusicEncounterManager
  (params (Music%20Encounter%20Manager :type Music%20Encounter%20Manager%20Component%20Object%20Reference))
  (Utilities|Array|Clear (Variables|Default|GetMusic%20Encounter%20Viewmodels))
  (bind _music_encounter_manager (Variables|GetMusic%20Encounter%20Manager)
    (:then
      (MusicEncounterManager|BindEventtoOnMusicEncounterConstructed _music_encounter_manager (EventDispatchers|CreateEvent :!event Create%20Music%20Encounter%20Viewmodel))
      (bind (_keypin _valuepin) (Utilities|Map|ForEachLoop(Map) (Variables|/MusicEncounter/BP_MusicEncounterManager.BP_MusicEncounterManager_C|GetMusic%20Encounters (Variables|GetMusic%20Encounter%20Manager)))
        (:then
          (CallFunction|CreateMusicEncounterViewmodel _valuepin))
        (:CompletedPin
          (if (> (Utilities|Array|Length (Variables|Default|GetMusic%20Encounter%20Viewmodels)) 0)
            (Variables|Default|SetSelected%20Music%20Encounter%20Viewmodel (Utilities|Array|Get(acopy) (Variables|Default|GetMusic%20Encounter%20Viewmodels)))
            (else
              (Variables|Default|SetSelected%20Music%20Encounter%20Viewmodel 0))))))
    (:else)))
