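; Viewmodel for the Music Encounter Dashboard tool that mirrors one music encounter for UI binding.
; SetMusicEncounter unbinds and clears the previous encounter, binds the new one's state and component events, then seeds tag, state and child viewmodels.
; Music encounter and audio component viewmodels live in maps keyed by component, with FieldNotify broadcasts on change.
; Getters expose the display name, interface checks and individual flags from the bitmask state.
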(fn AddAudioComponentViewmodel
  (params (Audio%20Component :type Audio%20Component%20Object%20Reference))
  (Utilities|IsValid Audio%20Component
    (:Is%20Valid
      (bind _returnvalue (Game|ConstructObjectfromClass "/MusicEncounter/Tools/MusicEncounterDashboard/Viewmodels/BP_MusicEncounterAudioComponentViewModel.BP_MusicEncounterAudioComponentViewModel_C"))
      (Class|MusicEncounterAudioComponentViewmodel|SetAudioComponent _returnvalue Audio%20Component)
      (Utilities|Map|Add (Variables|Default|GetAudio%20Component%20Viewmodels) (Variables|GetAudio%20Component) _returnvalue)
      (FieldNotification|BroadcastFieldValueChanged self "(FieldName=\"Get Audio Component Viewmodels\")"))))

(fn AddMusicEncounterComponentViewmodel
  (params (Music%20Encounter%20Component :type Music%20Encounter%20Component%20Base%20Object%20Reference))
  (Utilities|IsValid Music%20Encounter%20Component
    (:Is%20Valid
      (bind _returnvalue (Game|ConstructObjectfromClass "/MusicEncounter/Tools/MusicEncounterDashboard/Viewmodels/BP_MusicEncounterComponentViewModel.BP_MusicEncounterComponentViewModel_C"))
      (Class|MusicEncounterComponentViewModel|SetMusicEncounterComponent _returnvalue (Variables|GetMusic%20Encounter%20Component))
      (Utilities|Map|Add (Variables|Default|GetMusic%20Encounter%20Component%20Viewmodels) (Variables|GetMusic%20Encounter%20Component) _returnvalue)
      (FieldNotification|BroadcastFieldValueChanged self "(FieldName=\"Get Music Encounter Component Viewmodels\")"))))

(fn GetAudioComponentViewmodels
  (bind _values (Utilities|Map|Values (Variables|Default|GetAudio%20Component%20Viewmodels)))
  (return _values))

(fn GetBitFlag
  (params (InByte :type Byte))
  (return (Math|Float|Round (Math|Float|Power 2.0 (Math|Conversions|ToFloat(Byte) InByte)))))

(fn GetMusicEncounter
  (return (Variables|MusicEncounter|GetMusic%20Encounter)))

(fn GetMusicEncounterComponentViewmodels
  (bind _values (Utilities|Map|Values (Variables|Default|GetMusic%20Encounter%20Component%20Viewmodels)))
  (return _values))

(fn GetMusicEncounterDisplayName
  (return (Utilities|Text|ToText(String) (Utilities|GetDisplayName (Variables|MusicEncounter|GetMusic%20Encounter)))))

(fn GetMusicEncounterState
  (return (Variables|MusicEncounter|GetMusic%20Encounter%20State)))

(fn GetMusicEncounterStateValue
  (params (State%20Type :type EMusicEncounterState%20Enum :default "NewEnumerator8"))
  (return (Math|Conversions|ToBoolean(Integer) (Math|Integer|BitwiseAND (Default|GetBitFlag State%20Type) (Variables|MusicEncounter|GetMusic%20Encounter%20State)))))

(fn ImplementsInterface
  (params (Interface :type Interface%20Interface))
  (return (Utilities|DoesObjectImplementInterface (Variables|MusicEncounter|GetMusic%20Encounter) Interface)))

(fn RemoveAudioComponentViewmodel
  (params (Audio%20Component :type Audio%20Component%20Object%20Reference))
  (Utilities|IsValid Audio%20Component
    (:Is%20Valid
      (bind _returnvalue (Utilities|Map|Remove (Variables|Default|GetAudio%20Component%20Viewmodels) (Variables|GetAudio%20Component)))
      (if _returnvalue
        (FieldNotification|BroadcastFieldValueChanged self "(FieldName=\"Get Audio Component Viewmodels\")")))))

(fn RemoveMusicEncounterComponentViewmodel
  (params (Music%20Encounter%20Component :type Music%20Encounter%20Component%20Base%20Object%20Reference))
  (Utilities|IsValid Music%20Encounter%20Component
    (:Is%20Valid
      (bind _returnvalue (Utilities|Map|Remove (Variables|Default|GetMusic%20Encounter%20Component%20Viewmodels) (Variables|GetMusic%20Encounter%20Component)))
      (if _returnvalue
        (FieldNotification|BroadcastFieldValueChanged self "(FieldName=\"Get Music Encounter Component Viewmodels\")")))))

(fn SetMusicEncounter
  (params (Music%20Encounter :type Music%20Encounter%20Base%20Object%20Reference))
  (Utilities|FlowControl|Sequence
    (:then_0
      (Variables|MusicEncounter|GetMusic%20Encounter
        (:then
          (Utilities|FlowControl|Sequence
            (:then_0
              (MusicEncounter|UnbindEventfromOnMusicEncounterStateChanged (Variables|MusicEncounter|GetMusic%20Encounter) (EventDispatchers|CreateEvent :!event Set%20Music%20Encounter%20State))
              (MusicEncounter|UnbindEventfromOnMusicEncounterComponentAdded (Variables|MusicEncounter|GetMusic%20Encounter) (EventDispatchers|CreateEvent :!event Add%20Music%20Encounter%20Component%20Viewmodel))
              (MusicEncounter|UnbindEventfromOnMusicEncounterComponentRemoved (Variables|MusicEncounter|GetMusic%20Encounter) (EventDispatchers|CreateEvent :!event Remove%20Music%20Encounter%20Component%20Viewmodel))
              (MusicEncounter|UnbindEventfromOnMusicAudioComponentAdded (Variables|MusicEncounter|GetMusic%20Encounter) (EventDispatchers|CreateEvent :!event Add%20Audio%20Component%20Viewmodel))
              (MusicEncounter|UnbindEventfromOnMusicAudioComponentRemoved (Variables|MusicEncounter|GetMusic%20Encounter) (EventDispatchers|CreateEvent :!event Remove%20Audio%20Component%20Viewmodel)))
            (:then_1
              (Variables|MusicEncounter|SetInstance%20Tag)
              (CallFunction|SetMusicEncounterState 0)
              (Utilities|Map|Clear (Variables|Default|GetAudio%20Component%20Viewmodels))
              (Utilities|Map|Clear (Variables|Default|GetMusic%20Encounter%20Component%20Viewmodels))
              (FieldNotification|BroadcastFieldValueChanged self "(FieldName=\"Get Audio Components\")")
              (FieldNotification|BroadcastFieldValueChanged self "(FieldName=\"Get Music Encounter Component Viewmodels\")"))))
        (:else)))
    (:then_1
      (Variables|MusicEncounter|SetMusic%20Encounter Music%20Encounter)
      (Utilities|FlowControl|Sequence
        (:then_0
          (MusicEncounter|BindEventtoOnMusicEncounterStateChanged (Variables|MusicEncounter|GetMusic%20Encounter) (EventDispatchers|CreateEvent :!event Set%20Music%20Encounter%20State))
          (MusicEncounter|BindEventtoOnMusicEncounterComponentAdded (Variables|MusicEncounter|GetMusic%20Encounter) (EventDispatchers|CreateEvent :!event Add%20Music%20Encounter%20Component%20Viewmodel))
          (MusicEncounter|BindEventtoOnMusicEncounterComponentRemoved (Variables|MusicEncounter|GetMusic%20Encounter) (EventDispatchers|CreateEvent :!event Remove%20Music%20Encounter%20Component%20Viewmodel))
          (MusicEncounter|BindEventtoOnMusicAudioComponentAdded (Variables|MusicEncounter|GetMusic%20Encounter) (EventDispatchers|CreateEvent :!event Add%20Audio%20Component%20Viewmodel))
          (MusicEncounter|BindEventtoOnMusicAudioComponentRemoved (Variables|MusicEncounter|GetMusic%20Encounter) (EventDispatchers|CreateEvent :!event Remove%20Audio%20Component%20Viewmodel)))
        (:then_1
          (Utilities|FlowControl|Sequence
            (:then_0
              (Variables|MusicEncounter|SetInstance%20Tag (MusicEncounter|GetMusicEncounterInstanceTag (Variables|MusicEncounter|GetMusic%20Encounter))))
            (:then_1
              (Utilities|FlowControl|Sequence
                (:then_0
                  (CallFunction|SetMusicEncounterState (MusicEncounter|GetMusicEncounterState (Variables|MusicEncounter|GetMusic%20Encounter))))
                (:then_1
                  (Utilities|FlowControl|Sequence
                    (:then_0
                      (for _array_element (MusicEncounter|GetMusicEncounterComponents (Variables|MusicEncounter|GetMusic%20Encounter))
                        (Default|AddMusicEncounterComponentViewmodel _array_element)))
                    (:then_1
                      (Utilities|FlowControl|Sequence
                        (:then_0
                          (for _array_element_2 (MusicEncounter|GetMusicAudioComponents (Variables|MusicEncounter|GetMusic%20Encounter))
                            (Default|AddAudioComponentViewmodel _array_element_2)))))))))))))))

(fn SetMusicEncounterState
  (params (Music%20Encounter%20State :type Integer))
  (Variables|MusicEncounter|SetMusic%20Encounter%20State Music%20Encounter%20State))
